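# Animation Blueprint for the Bengal Basher NPC body, adding tail animation to a pose copied from the parent mesh.
# A Tail state machine plays borrowed Neon Cat tail sequences for idle, jump, crouch, aiming and DBNO, layered from the root joint.
# The result runs through the Suit's Control Rig, an additive yaw offset on the tail base, then RigidBody simulation using the Suit's physics asset.
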AnimBlueprint /BRCosmetics/Characters/Player/Male/Large/Bodies/M_LRG_BengalBasher_NPC/Meshes/M_LRG_BengalBasher_NPC_AnimBP
target_skeleton=&/M_LRG_BengalBasher_NPC_Skeleton
graphs [
 AnimGraph {nodes [
   AnimGraphNode_StateMachine_0 = Tail() {is=title AnimGraphNode_StateMachine}
   n1 = Control Rig(Source=Layered blend per bone(BasePose=Copy Pose From Mesh() {bUseAttachedParent=true}, BlendPoses_0=@AnimGraphNode_StateMachine_0) {layerSetup [{branchFilters [{boneName=C_Root_Main_Root_Jnt}]}]}) {
    alphaBoolBlend {blendOption=Linear}
    controlRigAssetReference {blueprintRigClass=/BRCosmetics/Characters/Player/Male/Large/Bodies/M_LRG_BengalBasher_Suit/Meshes/M_LRG_BengalBasher_Suit_CtrlRig.M_LRG_BengalBasher_Suit_CtrlRig_C}
   }
   n2 = "Transform (Modify) Bone - Bone: C_Tail_A_Base_Jnt"(ComponentPose=Local To Component(LocalPose=@n1), Alpha=1, Translation="0,0,0", Rotation="0,30,0", Scale="1,1,1") {
    rotationMode=BMM_Additive
    rotationSpace=BCS_BoneSpace
    alphaBoolBlend {blendOption=Linear}
    boneToModify {boneName=C_Tail_A_Base_Jnt}
    rotation {yaw=30}
   }
   n3 = RigidBody(ComponentPose=@n2) {
    bOverrideWorldGravity=true
    overridePhysicsAsset=/BRCosmetics/Characters/Player/Male/Large/Bodies/M_LRG_BengalBasher_Suit/Meshes/M_LRG_BengalBasher_Suit_Physics
    alphaBoolBlend {blendOption=Linear}
    componentAppliedLinearAccClamp=<2000 2000 2000>
    componentLinearAccScale=<0.2 0.2 0.3>
    componentLinearVelScale=<0.2 0.2 0.3>
    overrideWorldGravity {z=-980}
   }
   Output Pose(Result=Component To Local(ComponentPose=@n3)) {layerGroup=None}
  ]}
 Tail {
  entry=idle
  path=AnimGraph.AnimGraphNode_StateMachine_0.Tail
  ids=id
  nodes [
   AnimStateConduitNode_1 = IDS() {is=title AnimStateConduitNode}
   AnimStateConduitNode_2 = Conduit_DownSightsToIdle() {is=title AnimStateConduitNode}
   AnimStateNode_0 = JumpLand() {is=title AnimStateNode, stateLeft {guid=#1, notifyName=JumpLandExit}}
   AnimStateNode_12 = DownSights() {is=title AnimStateNode}
   AnimStateNode_13 = IdleToDownSights() {is=title AnimStateNode}
   AnimStateNode_18 = Crouching() {is=title AnimStateNode}
   AnimStateNode_20 = idle() {is=title AnimStateNode}
   AnimStateNode_4 = StartJump() {is=title AnimStateNode}
   AnimStateNode_5 = JumpLoop() {is=title AnimStateNode}
   AnimStateNode_6 = JumpFall() {is=title AnimStateNode}
   AnimStateNode_9 = DBNO() {is=title AnimStateNode}
   AnimStateTransitionNode_0 = AnimStateTransitionNode()
   AnimStateTransitionNode_10 = AnimStateTransitionNode()
   AnimStateTransitionNode_11 = AnimStateTransitionNode()
   AnimStateTransitionNode_12 = AnimStateTransitionNode() {bAutomaticRuleBasedOnSequencePlayerInState=true}
   AnimStateTransitionNode_13 = AnimStateTransitionNode()
   AnimStateTransitionNode_14 = AnimStateTransitionNode()
   AnimStateTransitionNode_15 = AnimStateTransitionNode()
   AnimStateTransitionNode_20 = AnimStateTransitionNode()
   AnimStateTransitionNode_22 = AnimStateTransitionNode()
   AnimStateTransitionNode_23 = AnimStateTransitionNode()
   AnimStateTransitionNode_24 = AnimStateTransitionNode()
   AnimStateTransitionNode_26 = AnimStateTransitionNode()
   AnimStateTransitionNode_30 = AnimStateTransitionNode()
   AnimStateTransitionNode_31 = AnimStateTransitionNode()
   AnimStateTransitionNode_32 = AnimStateTransitionNode()
   AnimStateTransitionNode_33 = AnimStateTransitionNode()
   AnimStateTransitionNode_34 = AnimStateTransitionNode()
   AnimStateTransitionNode_35 = AnimStateTransitionNode()
   AnimStateTransitionNode_37 = AnimStateTransitionNode()
   AnimStateTransitionNode_38 = AnimStateTransitionNode()
   AnimStateTransitionNode_39 = AnimStateTransitionNode() {bAutomaticRuleBasedOnSequencePlayerInState=true}
   AnimStateTransitionNode_4 = AnimStateTransitionNode() {bAutomaticRuleBasedOnSequencePlayerInState=true}
   AnimStateTransitionNode_40 = AnimStateTransitionNode()
   AnimStateTransitionNode_5 = AnimStateTransitionNode()
   AnimStateTransitionNode_6 = AnimStateTransitionNode() {bAutomaticRuleBasedOnSequencePlayerInState=true}
   AnimStateTransitionNode_7 = AnimStateTransitionNode()
   AnimStateTransitionNode_8 = AnimStateTransitionNode()
  ]
  links [
   {
    dst=IdleToDownSights
    rule=AnimGraph.AnimGraphNode_StateMachine_0.Tail.AnimStateTransitionNode_24.Transition
    src=Conduit_DownSightsToIdle
    via=AnimStateTransitionNode_24
   }
   {
    dst=idle
    rule=AnimGraph.AnimGraphNode_StateMachine_0.Tail.AnimStateTransitionNode_26.Transition
    src=Conduit_DownSightsToIdle
    via=AnimStateTransitionNode_26
   }
   {
    dst=idle
    rule=AnimGraph.AnimGraphNode_StateMachine_0.Tail.AnimStateTransitionNode_33.Transition
    src=Conduit_DownSightsToIdle
    via=AnimStateTransitionNode_33
   }
   {
    dst=StartJump
    rule=AnimGraph.AnimGraphNode_StateMachine_0.Tail.AnimStateTransitionNode_10.Transition
    src=Crouching
    via=AnimStateTransitionNode_10
   }
   {
    dst=idle
    rule=AnimGraph.AnimGraphNode_StateMachine_0.Tail.AnimStateTransitionNode_34.Transition
    src=Crouching
    via=AnimStateTransitionNode_34
   }
   {dst=idle, rule=AnimGraph.AnimGraphNode_StateMachine_0.Tail.AnimStateTransitionNode_14.Transition, src=DBNO, via=AnimStateTransitionNode_14}
   {
    dst=Crouching
    rule=AnimGraph.AnimGraphNode_StateMachine_0.Tail.AnimStateTransitionNode_11.Transition
    src=DownSights
    via=AnimStateTransitionNode_11
   }
   {
    dst=StartJump
    rule=AnimGraph.AnimGraphNode_StateMachine_0.Tail.AnimStateTransitionNode_22.Transition
    src=DownSights
    via=AnimStateTransitionNode_22
   }
   {
    dst=StartJump
    rule=AnimGraph.AnimGraphNode_StateMachine_0.Tail.AnimStateTransitionNode_31.Transition
    src=DownSights
    via=AnimStateTransitionNode_31
   }
   {
    dst=IdleToDownSights
    rule=AnimGraph.AnimGraphNode_StateMachine_0.Tail.AnimStateTransitionNode_23.Transition
    src=IDS
    via=AnimStateTransitionNode_23
   }
   {
    dst=DownSights
    rule=AnimGraph.AnimGraphNode_StateMachine_0.Tail.AnimStateTransitionNode_12.Transition
    src=IdleToDownSights
    via=AnimStateTransitionNode_12
   }
   {
    dst=DownSights
    rule=AnimGraph.AnimGraphNode_StateMachine_0.Tail.AnimStateTransitionNode_35.Transition
    src=IdleToDownSights
    via=AnimStateTransitionNode_35
   }
   {
    dst=JumpLand
    rule=AnimGraph.AnimGraphNode_StateMachine_0.Tail.AnimStateTransitionNode_4.Transition
    src=JumpFall
    via=AnimStateTransitionNode_4
   }
   {
    dst=StartJump
    rule=AnimGraph.AnimGraphNode_StateMachine_0.Tail.AnimStateTransitionNode_8.Transition
    src=JumpFall
    via=AnimStateTransitionNode_8
   }
   {
    dst=StartJump
    rule=AnimGraph.AnimGraphNode_StateMachine_0.Tail.AnimStateTransitionNode_20.Transition
    src=JumpLand
    via=AnimStateTransitionNode_20
   }
   {
    dst=idle
    rule=AnimGraph.AnimGraphNode_StateMachine_0.Tail.AnimStateTransitionNode_37.Transition
    src=JumpLand
    via=AnimStateTransitionNode_37
   }
   {
    dst=idle
    rule=AnimGraph.AnimGraphNode_StateMachine_0.Tail.AnimStateTransitionNode_38.Transition
    src=JumpLand
    via=AnimStateTransitionNode_38
   }
   {
    dst=idle
    rule=AnimGraph.AnimGraphNode_StateMachine_0.Tail.AnimStateTransitionNode_39.Transition
    src=JumpLand
    via=AnimStateTransitionNode_39
   }
   {
    dst=JumpFall
    rule=AnimGraph.AnimGraphNode_StateMachine_0.Tail.AnimStateTransitionNode_30.Transition
    src=JumpLoop
    via=AnimStateTransitionNode_30
   }
   {
    dst=JumpFall
    rule=AnimGraph.AnimGraphNode_StateMachine_0.Tail.AnimStateTransitionNode_5.Transition
    src=JumpLoop
    via=AnimStateTransitionNode_5
   }
   {
    dst=JumpFall
    rule=AnimGraph.AnimGraphNode_StateMachine_0.Tail.AnimStateTransitionNode_40.Transition
    src=StartJump
    via=AnimStateTransitionNode_40
   }
   {
    dst=JumpFall
    rule=AnimGraph.AnimGraphNode_StateMachine_0.Tail.AnimStateTransitionNode_7.Transition
    src=StartJump
    via=AnimStateTransitionNode_7
   }
   {
    dst=JumpLoop
    rule=AnimGraph.AnimGraphNode_StateMachine_0.Tail.AnimStateTransitionNode_6.Transition
    src=StartJump
    via=AnimStateTransitionNode_6
   }
   {
    dst=Crouching
    rule=AnimGraph.AnimGraphNode_StateMachine_0.Tail.AnimStateTransitionNode_32.Transition
    src=idle
    via=AnimStateTransitionNode_32
   }
   {dst=DBNO, rule=AnimGraph.AnimGraphNode_StateMachine_0.Tail.AnimStateTransitionNode_13.Transition, src=idle, via=AnimStateTransitionNode_13}
   {dst=IDS, rule=AnimGraph.AnimGraphNode_StateMachine_0.Tail.AnimStateTransitionNode_15.Transition, src=idle, via=AnimStateTransitionNode_15}
   {
    dst=StartJump
    rule=AnimGraph.AnimGraphNode_StateMachine_0.Tail.AnimStateTransitionNode_0.Transition
    src=idle
    via=AnimStateTransitionNode_0
   }
  ]
 }
 IDS {path=AnimGraph.AnimGraphNode_StateMachine_0.Tail.AnimStateConduitNode_1.IDS, nodes [Result(bCanEnterTransition=true)]}
 Conduit_DownSightsToIdle {path=AnimGraph.AnimGraphNode_StateMachine_0.Tail.AnimStateConduitNode_2.Conduit_DownSightsToIdle, nodes [Result(bCanEnterTransition=true)]}
 JumpLand {
  path=AnimGraph.AnimGraphNode_StateMachine_0.Tail.AnimStateNode_0.JumpLand
  nodes [
   n1 = Blend Poses by int!(
    BlendPose_0:Pose Link Structure
    BlendPose_1:Pose Link Structure
    BlendTime_0:float=0.1
    BlendTime_1:float=0.1
    ActiveChildIndex:int=Get JumpIndex() -> (JumpIndex:int=0)
   ) -> (Pose:Pose Link Structure) {
    is=title AnimGraphNode_BlendListByInt
    blendType=Linear
    blendPose [{linkId=-1, sourceLinkId=-1}, {linkId=-1, sourceLinkId=-1}]
    blendTime [0.1,0.1]
   }
   Blend Poses by bool(BlendPose_0=@n1, bActiveValue=Get bIsMoving2D()) {blendType=Linear}
   Output Animation Pose(Result=Sequence Player 'Tail_Jump_Land'() {sequence=/Game/Accessories/FORT_Tails/Tail_Neon_Cat/Anims/Tail_Jump_Land})
  ]
 }
 DownSights {
  path=AnimGraph.AnimGraphNode_StateMachine_0.Tail.AnimStateNode_12.DownSights
  nodes [Output Animation Pose(Result=Sequence Player 'Tail_Downsights'() {sequence=/Game/Accessories/FORT_Tails/Tail_Neon_Cat/Anims/Tail_Downsights})]
 }
 IdleToDownSights {
  path=AnimGraph.AnimGraphNode_StateMachine_0.Tail.AnimStateNode_13.IdleToDownSights
  nodes [Output Animation Pose(Result=Sequence Player 'Tail_Downsights'() {sequence=&Tail_Downsights})]
 }
 Crouching {
  path=AnimGraph.AnimGraphNode_StateMachine_0.Tail.AnimStateNode_18.Crouching
  nodes [
   Blend Poses by bool(bActiveValue=Get bIsTargeting()) {blendType=Linear}
   AnimGraphNode_StateMachine_0 = Crouch() {is=title AnimGraphNode_StateMachine}
   Output Animation Pose(Result=@AnimGraphNode_StateMachine_0)
  ]
 }
 Crouch {
  entry=Blank
  path=AnimGraph.AnimGraphNode_StateMachine_0.Tail.AnimStateNode_18.Crouching.AnimGraphNode_StateMachine_0.Crouch
  ids=id
  nodes [
   AnimStateNode_0 = Blank() {is=title AnimStateNode}
   AnimStateNode_2 = CrouchSprinting() {is=title AnimStateNode}
   AnimStateNode_27 = CrouchWalk() {is=title AnimStateNode}
   AnimStateNode_29 = Crouch() {is=title AnimStateNode}
   AnimStateTransitionNode_11 = AnimStateTransitionNode()
   AnimStateTransitionNode_12 = AnimStateTransitionNode()
   AnimStateTransitionNode_4 = AnimStateTransitionNode()
   AnimStateTransitionNode_47 = AnimStateTransitionNode()
   AnimStateTransitionNode_55 = AnimStateTransitionNode()
   AnimStateTransitionNode_7 = AnimStateTransitionNode()
  ]
  links [
   {
    dst=Crouch
    rule=AnimGraph.AnimGraphNode_StateMachine_0.Tail.AnimStateNode_18.Crouching.AnimGraphNode_StateMachine_0.Crouch.AnimStateTransitionNode_11.Transition
    src=Blank
    via=AnimStateTransitionNode_11
   }
   {
    dst=CrouchWalk
    rule=AnimGraph.AnimGraphNode_StateMachine_0.Tail.AnimStateNode_18.Crouching.AnimGraphNode_StateMachine_0.Crouch.AnimStateTransitionNode_12.Transition
    src=Blank
    via=AnimStateTransitionNode_12
   }
   {
    dst=CrouchWalk
    rule=AnimGraph.AnimGraphNode_StateMachine_0.Tail.AnimStateNode_18.Crouching.AnimGraphNode_StateMachine_0.Crouch.AnimStateTransitionNode_7.Transition
    src=Crouch
    via=AnimStateTransitionNode_7
   }
   {
    dst=CrouchWalk
    rule=AnimGraph.AnimGraphNode_StateMachine_0.Tail.AnimStateNode_18.Crouching.AnimGraphNode_StateMachine_0.Crouch.AnimStateTransitionNode_55.AnimationTransitionGraph_1
    src=CrouchSprinting
    via=AnimStateTransitionNode_55
   }
   {
    dst=Crouch
    rule=AnimGraph.AnimGraphNode_StateMachine_0.Tail.AnimStateNode_18.Crouching.AnimGraphNode_StateMachine_0.Crouch.AnimStateTransitionNode_4.Transition
    src=CrouchWalk
    via=AnimStateTransitionNode_4
   }
   {
    dst=CrouchSprinting
    rule=AnimGraph.AnimGraphNode_StateMachine_0.Tail.AnimStateNode_18.Crouching.AnimGraphNode_StateMachine_0.Crouch.AnimStateTransitionNode_47.AnimationTransitionGraph_1
    src=CrouchWalk
    via=AnimStateTransitionNode_47
   }
  ]
 }
 Blank {
  path=AnimGraph.AnimGraphNode_StateMachine_0.Tail.AnimStateNode_18.Crouching.AnimGraphNode_StateMachine_0.Crouch.AnimStateNode_0.Blank
  nodes [Output Animation Pose()]
 }
 CrouchSprinting {
  path=AnimGraph.AnimGraphNode_StateMachine_0.Tail.AnimStateNode_18.Crouching.AnimGraphNode_StateMachine_0.Crouch.AnimStateNode_2.CrouchSprinting
  nodes [Output Animation Pose(Result=Sequence Player 'Tail_Crouch_Sprint'() {sequence=/Game/Accessories/FORT_Tails/Tail_Neon_Cat/Anims/Tail_Crouch_Sprint})]
 }
 CrouchWalk {
  path=AnimGraph.AnimGraphNode_StateMachine_0.Tail.AnimStateNode_18.Crouching.AnimGraphNode_StateMachine_0.Crouch.AnimStateNode_27.CrouchWalk
  nodes [Output Animation Pose(Result=Sequence Player 'Tail_Crouch_Walk'() {sequence=/Game/Accessories/FORT_Tails/Tail_Neon_Cat/Anims/Tail_Crouch_Walk})]
 }
 Crouch {
  path=AnimGraph.AnimGraphNode_StateMachine_0.Tail.AnimStateNode_18.Crouching.AnimGraphNode_StateMachine_0.Crouch.AnimStateNode_29.Crouch
  nodes [Output Animation Pose(Result=Sequence Player 'Tail_Crouch_Idle'() {sequence=/Game/Accessories/FORT_Tails/Tail_Neon_Cat/Anims/Tail_Crouch_Idle})]
 }
 Transition {
  path=AnimGraph.AnimGraphNode_StateMachine_0.Tail.AnimStateNode_18.Crouching.AnimGraphNode_StateMachine_0.Crouch.AnimStateTransitionNode_11.Transition
  nodes [Result(bCanEnterTransition=NOT Boolean(A=Get bIsMoving2D()))]
 }
 Transition {
  path=AnimGraph.AnimGraphNode_StateMachine_0.Tail.AnimStateNode_18.Crouching.AnimGraphNode_StateMachine_0.Crouch.AnimStateTransitionNode_12.Transition
  nodes [Result(bCanEnterTransition=Get bIsMoving2D())]
 }
 Transition {
  path=AnimGraph.AnimGraphNode_StateMachine_0.Tail.AnimStateNode_18.Crouching.AnimGraphNode_StateMachine_0.Crouch.AnimStateTransitionNode_4.Transition
  nodes [Result(bCanEnterTransition=NOT Boolean(A=Get bIsMoving2D()))]
 }
 AnimationTransitionGraph_1 {
  path=AnimGraph.AnimGraphNode_StateMachine_0.Tail.AnimStateNode_18.Crouching.AnimGraphNode_StateMachine_0.Crouch.AnimStateTransitionNode_47.AnimationTransitionGraph_1
  nodes [Result(bCanEnterTransition=Get bIsCrouchSprinting())]
 }
 AnimationTransitionGraph_1 {
  path=AnimGraph.AnimGraphNode_StateMachine_0.Tail.AnimStateNode_18.Crouching.AnimGraphNode_StateMachine_0.Crouch.AnimStateTransitionNode_55.AnimationTransitionGraph_1
  nodes [Result(bCanEnterTransition=NOT Boolean(A=Get bIsCrouchSprinting()))]
 }
 Transition {
  path=AnimGraph.AnimGraphNode_StateMachine_0.Tail.AnimStateNode_18.Crouching.AnimGraphNode_StateMachine_0.Crouch.AnimStateTransitionNode_7.Transition
  nodes [Result(bCanEnterTransition=Get bIsMoving2D())]
 }
 idle {
  path=AnimGraph.AnimGraphNode_StateMachine_0.Tail.AnimStateNode_20.idle
  nodes [Output Animation Pose(Result=Sequence Player 'Tail_Idle'() {sequence=/Game/Accessories/FORT_Tails/Tail_Neon_Cat/Anims/Tail_Idle})]
 }
 StartJump {
  path=AnimGraph.AnimGraphNode_StateMachine_0.Tail.AnimStateNode_4.StartJump
  nodes [
   n1 = Blend Poses by int(BlendTime_0=0.1, BlendTime_1=0.1, ActiveChildIndex=Get JumpIndex() -> (JumpIndex=0)) {blendType=Linear, blendPose [{linkId=-1, sourceLinkId=-1}, {linkId=-1, sourceLinkId=-1}], blendTime [0.1,0.1]}
   Blend Poses by bool(BlendPose_0=@n1, bActiveValue=Get bIsMoving2D()) {blendType=Linear}
   Output Animation Pose(Result=Sequence Player 'Tail_Jump_Up'() {sequence=/Game/Accessories/FORT_Tails/Tail_Neon_Cat/Anims/Tail_Jump_Up})
  ]
 }
 JumpLoop {
  path=AnimGraph.AnimGraphNode_StateMachine_0.Tail.AnimStateNode_5.JumpLoop
  nodes [
   n1 = Blend Poses by int(BlendTime_0=0.1, BlendTime_1=0.1, ActiveChildIndex=Get JumpIndex() -> (JumpIndex=0)) {blendType=Linear, blendPose [{linkId=-1, sourceLinkId=-1}, {linkId=-1, sourceLinkId=-1}], blendTime [0.1,0.1]}
   Blend Poses by bool(BlendPose_0=@n1, bActiveValue=Get bIsMoving2D()) {blendType=Linear}
   Output Animation Pose(Result=Sequence Player 'Tail_Jump_Loop_Apex_01'() {sequence=/Game/Accessories/FORT_Tails/Tail_Neon_Cat/Anims/Tail_Jump_Loop_Apex_01})
  ]
 }
 JumpFall {
  path=AnimGraph.AnimGraphNode_StateMachine_0.Tail.AnimStateNode_6.JumpFall
  nodes [
   n1 = Blend Poses by int(BlendTime_0=0.1, BlendTime_1=0.1, ActiveChildIndex=Get JumpIndex() -> (JumpIndex=0)) {blendType=Linear, blendPose [{linkId=-1, sourceLinkId=-1}, {linkId=-1, sourceLinkId=-1}], blendTime [0.1,0.1]}
   Blend Poses by bool(BlendPose_0=@n1, bActiveValue=Get bIsMoving2D()) {blendType=Linear}
   Output Animation Pose(Result=Sequence Player 'Tail_Jump_Fall'() {bLoopAnimation=false, sequence=/Game/Accessories/FORT_Tails/Tail_Neon_Cat/Anims/Tail_Jump_Fall})
  ]
 }
 DBNO {
  path=AnimGraph.AnimGraphNode_StateMachine_0.Tail.AnimStateNode_9.DBNO
  nodes [Output Animation Pose(Result=Sequence Player 'Tail_DBNO'() {sequence=/Game/Accessories/FORT_Tails/Tail_Neon_Cat/Anims/Tail_DBNO})]
 }
 Transition {path=AnimGraph.AnimGraphNode_StateMachine_0.Tail.AnimStateTransitionNode_0.Transition, nodes [Result(bCanEnterTransition=Get bIsJumping())]}
 Transition {path=AnimGraph.AnimGraphNode_StateMachine_0.Tail.AnimStateTransitionNode_10.Transition, nodes [Result(bCanEnterTransition=Get bIsJumping())]}
 Transition {
  path=AnimGraph.AnimGraphNode_StateMachine_0.Tail.AnimStateTransitionNode_11.Transition
  nodes [Result(bCanEnterTransition=Get bIsCrouching())]
 }
 Transition {path=AnimGraph.AnimGraphNode_StateMachine_0.Tail.AnimStateTransitionNode_12.Transition, nodes [Result()]}
 Transition {
  path=AnimGraph.AnimGraphNode_StateMachine_0.Tail.AnimStateTransitionNode_13.Transition
  nodes [
   n1 = Get bIsDBNO()
   Result(bCanEnterTransition=@n1)
   NOT Boolean(A=@n1)
  ]
 }
 Transition {
  path=AnimGraph.AnimGraphNode_StateMachine_0.Tail.AnimStateTransitionNode_14.Transition
  nodes [Result(bCanEnterTransition=NOT Boolean(A=Get bIsDBNO()))]
 }
 Transition {
  path=AnimGraph.AnimGraphNode_StateMachine_0.Tail.AnimStateTransitionNode_15.Transition
  nodes [Result(bCanEnterTransition=Get bIsTargeting())]
 }
 Transition {path=AnimGraph.AnimGraphNode_StateMachine_0.Tail.AnimStateTransitionNode_20.Transition, nodes [Result(bCanEnterTransition=Get bIsJumping())]}
 Transition {path=AnimGraph.AnimGraphNode_StateMachine_0.Tail.AnimStateTransitionNode_22.Transition, nodes [Result(bCanEnterTransition=Get bIsJumping())]}
 Transition {
  path=AnimGraph.AnimGraphNode_StateMachine_0.Tail.AnimStateTransitionNode_23.Transition
  nodes [Result(bCanEnterTransition=Get bIsCrouching())]
 }
 Transition {
  path=AnimGraph.AnimGraphNode_StateMachine_0.Tail.AnimStateTransitionNode_24.Transition
  nodes [Result(bCanEnterTransition=Get bIsTargeting())]
 }
 Transition {
  path=AnimGraph.AnimGraphNode_StateMachine_0.Tail.AnimStateTransitionNode_26.Transition
  nodes [Result(bCanEnterTransition=Get bIsCrouching())]
 }
 Transition {
  path=AnimGraph.AnimGraphNode_StateMachine_0.Tail.AnimStateTransitionNode_30.Transition
  nodes [Result(bCanEnterTransition=Get bIsSlopeSliding() -> (bIsSlopeSliding:bool))]
 }
 Transition {path=AnimGraph.AnimGraphNode_StateMachine_0.Tail.AnimStateTransitionNode_31.Transition, nodes [Result(bCanEnterTransition=Get bIsFalling())]}
 Transition {
  path=AnimGraph.AnimGraphNode_StateMachine_0.Tail.AnimStateTransitionNode_32.Transition
  nodes [Result(bCanEnterTransition=Get bIsCrouching())]
 }
 Transition {
  path=AnimGraph.AnimGraphNode_StateMachine_0.Tail.AnimStateTransitionNode_33.Transition
  nodes [Result(bCanEnterTransition=Get bIsMoving2D())]
 }
 Transition {
  path=AnimGraph.AnimGraphNode_StateMachine_0.Tail.AnimStateTransitionNode_34.Transition
  nodes [Result(bCanEnterTransition=NOT Boolean(A=Get bIsCrouching()))]
 }
 Transition {
  path=AnimGraph.AnimGraphNode_StateMachine_0.Tail.AnimStateTransitionNode_35.Transition
  nodes [Result(bCanEnterTransition=NOT Boolean(A=Get bIsTargeting()))]
 }
 Transition {
  path=AnimGraph.AnimGraphNode_StateMachine_0.Tail.AnimStateTransitionNode_37.Transition
  nodes [Result(bCanEnterTransition=Get bIsSlopeSliding())]
 }
 Transition {
  path=AnimGraph.AnimGraphNode_StateMachine_0.Tail.AnimStateTransitionNode_38.Transition
  nodes [Result(bCanEnterTransition=Get bIsSlopeSliding())]
 }
 Transition {path=AnimGraph.AnimGraphNode_StateMachine_0.Tail.AnimStateTransitionNode_39.Transition, nodes [Result()]}
 Transition {path=AnimGraph.AnimGraphNode_StateMachine_0.Tail.AnimStateTransitionNode_4.Transition, nodes [Result()]}
 Transition {
  path=AnimGraph.AnimGraphNode_StateMachine_0.Tail.AnimStateTransitionNode_40.Transition
  nodes [Result(bCanEnterTransition=Get bIsSlopeSliding())]
 }
 Transition {
  path=AnimGraph.AnimGraphNode_StateMachine_0.Tail.AnimStateTransitionNode_5.Transition
  nodes [Result(bCanEnterTransition=Get bIsOnGround() -> (bIsOnGround:bool))]
 }
 Transition {path=AnimGraph.AnimGraphNode_StateMachine_0.Tail.AnimStateTransitionNode_6.Transition, nodes [Result()]}
 Transition {path=AnimGraph.AnimGraphNode_StateMachine_0.Tail.AnimStateTransitionNode_7.Transition, nodes [Result(bCanEnterTransition=Get bIsOnGround())]}
 Transition {path=AnimGraph.AnimGraphNode_StateMachine_0.Tail.AnimStateTransitionNode_8.Transition, nodes [Result(bCanEnterTransition=Get bIsJumping())]}
 EventGraph {nodes [
   AnimNotify_RollRandomIdle() {is=title K2Node_Event}
   > Set RandomIdle(RandomIdle:int="Clamp (Integer)"(Value=RandomIntegerInRange!(Min:int=0, Max:int=11) -> (ReturnValue:int=0), Min=0, Max=2) -> (ReturnValue=0)) -> (Output_Get:int=0)
   Event BlueprintUpdateAnimation()
   > n1 = Branch(Condition=Get bIsSurfaceSwimming())
   > Set IsSwimming(IsSwimming:bool=true) -> (Output_Get:bool)
   Branch(execute=@n1.else, Condition=Get bIsOnGround())
   > Set IsSwimming(IsSwimming=false)
   AnimNotify_JumpLandExit() {is=title K2Node_Event}
   > Set JumpIndex(JumpIndex:int=RandomInteger!(Max:int=2) -> (ReturnValue:int=0)) -> (Output_Get:int=0)
   Do Once()
  ]}
]
variables [
 IsSwimming {type=boolean}
 JumpIndex {type=integer}
 RandomIdle {type=integer}
]
